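# Level Sequence for shot pnm1000 in PistolCanary, acting as a container that assembles the shot from per-department subsequences.
# It has no bindings of its own; parallel sub tracks cover animation, audio, camera, gameplay, lighting, techart and vfx, with vfx offset one frame earlier.
# Data layer tracks span the shot for World Partition content, covering S3, several ranges ending at S3, and a SeveredTentacle layer labelled Skip Section.
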LevelSequence /PistolCanary/Sequence/pnm/pnm1000/PistolCanary_pnm1000
bindings []
display_rate {denominator=1, numerator=60}
tick_resolution {denominator=1, numerator=24000}
tracks [
 {type=MovieSceneSubTrack, display_name=animation, sections [{type=MovieSceneSubSection, end_frame=4557, start_frame=3600}]}
 {type=MovieSceneSubTrack, display_name=audio, sections [{type=MovieSceneSubSection, end_frame=4557, start_frame=3600}]}
 {type=MovieSceneSubTrack, display_name=camera, sections [{type=MovieSceneSubSection, end_frame=4557, start_frame=3600}]}
 {type=MovieSceneSubTrack, display_name=gameplay, sections [{type=MovieSceneSubSection, end_frame=4557, start_frame=3600}]}
 {type=MovieSceneSubTrack, display_name=lighting, sections [{type=MovieSceneSubSection, end_frame=4557, start_frame=3600}]}
 {type=MovieSceneSubTrack, display_name=techart, sections [{type=MovieSceneSubSection, end_frame=4557, start_frame=3600}]}
 {type=MovieSceneSubTrack, display_name=vfx, sections [{type=MovieSceneSubSection, end_frame=4556, start_frame=3599}]}
 {type=MovieSceneDataLayerTrack, display_name=Data Layer - S3, sections [{type=MovieSceneDataLayerSection, end_frame=4556, start_frame=3600}]}
 {
  type=MovieSceneDataLayerTrack
  display_name=Data Layer - S0_to_End_S3
  sections [{type=MovieSceneDataLayerSection, end_frame=4556, start_frame=3600}]
 }
 {
  type=MovieSceneDataLayerTrack
  display_name=Data Layer - Day1_to_End_S3
  sections [{type=MovieSceneDataLayerSection, end_frame=4556, start_frame=3600}]
 }
 {
  type=MovieSceneDataLayerTrack
  display_name=Data Layer - S1_to_End_S3
  sections [{type=MovieSceneDataLayerSection, end_frame=4556, start_frame=3600}]
 }
 {
  type=MovieSceneDataLayerTrack
  display_name="Data Layer - SeveredTentacle (Skip Section)"
  sections [{type=MovieSceneDataLayerSection, end_frame=4556, start_frame=3600}]
 }
]
